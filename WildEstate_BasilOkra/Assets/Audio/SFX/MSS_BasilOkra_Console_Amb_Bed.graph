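# Looping 5.1 MetaSound Source for the BasilOkra console's ambience bed, attenuated and volume-modulated by the BasilOkra control bus.
# On play, an ADSR envelope starts the loop player and fades it in by multiplying each output channel.
# The Default page plays the 5.1 loop, while the Mobile page sends a mono loop to front left and right with a lower sustain.
# Both one-shot array inputs are unconnected in either graph.
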MetaSoundSource /WildEstate_BasilOkra/Assets/Audio/SFX/MSS_BasilOkra_Console_Amb_Bed
inputs [
 Amb_Loop_Mono {type=WaveAsset}
 Amb_Loop_Surr {type=WaveAsset}
 OneShots A {type=WaveAsset:Array}
 OneShots B {type=WaveAsset:Array}
 UE.Source.OnPlay {type=Trigger}
]
interfaces [UE.OutputFormat.5dot1:1.0, UE.Source:1.1]
outputs [
 "UE.OutputFormat.5dot1.Audio:0" {type=Audio}
 "UE.OutputFormat.5dot1.Audio:1" {type=Audio}
 "UE.OutputFormat.5dot1.Audio:2" {type=Audio}
 "UE.OutputFormat.5dot1.Audio:3" {type=Audio}
 "UE.OutputFormat.5dot1.Audio:4" {type=Audio}
 "UE.OutputFormat.5dot1.Audio:5" {type=Audio}
]
pages [
 Default {
  is_build_graph=true
  input_defaults {
   Amb_Loop_Mono=/WildEstate_BasilOkra/Assets/Audio/SFX/wav/BasilOkra_IP_Amb_Loop_Mono
   Amb_Loop_Surr=/WildEstate_BasilOkra/Assets/Audio/SFX/wav/BasilOkra_IP_Amb_Loop_5_1
   UE.Source.OnPlay=null
   OneShots A [
    /WildEstate_BasilOkra/Assets/Audio/SFX/wav/BasilOkra_Amb_OS_00
    /WildEstate_BasilOkra/Assets/Audio/SFX/wav/BasilOkra_Amb_OS_01
    /WildEstate_BasilOkra/Assets/Audio/SFX/wav/BasilOkra_Amb_OS_03
    /WildEstate_BasilOkra/Assets/Audio/SFX/wav/BasilOkra_Amb_OS_04
    /WildEstate_BasilOkra/Assets/Audio/SFX/wav/BasilOkra_Amb_OS_05
    /WildEstate_BasilOkra/Assets/Audio/SFX/wav/BasilOkra_Amb_OS_06
   ]
   OneShots B [
    /WildEstate_BasilOkra/Assets/Audio/SFX/wav/BasilOkra_Amb_OS_07
    /WildEstate_BasilOkra/Assets/Audio/SFX/wav/BasilOkra_Amb_OS_08
    /WildEstate_BasilOkra/Assets/Audio/SFX/wav/BasilOkra_Amb_OS_09
    /WildEstate_BasilOkra/Assets/Audio/SFX/wav/BasilOkra_Amb_OS_10
   ]
  }
  nodes [
   InitVariable.Trigger()
   InitVariable.Trigger()
   InitVariable.Trigger()
   InitVariable.Trigger()
   ADSREnvelopeADSREnvelope_0 = ADSR Envelope.ADSR Envelope:Audio(Trigger Attack=^UE.Source.OnPlay) {also {defaults {Attack Time=1, Sustain Level=1}}}
   n1 = UE.Reroute(Value=UE.Multiply:Audio by Float(PrimaryOperand=@ADSREnvelopeADSREnvelope_0.Out Envelope).Out)
   WavePlayer_0 = UE.Wave Player:5dot1(Wave Asset=^Amb_Loop_Surr, Play=@ADSREnvelopeADSREnvelope_0.On Attack Triggered) {also {defaults {Loop="<MetaSoundBuilderResult.SUCCEEDED: 0>"}, version [1,1]}}
   Multiply_0 = UE.Multiply:Audio(PrimaryOperand=@n1.Value, AdditionalOperands=@WavePlayer_0.Out Low Frequency)
   Multiply_1 = UE.Multiply:Audio(PrimaryOperand=@n1.Value, AdditionalOperands=@WavePlayer_0.Out Side Right)
   Multiply_2 = UE.Multiply:Audio(PrimaryOperand=@n1.Value, AdditionalOperands=@WavePlayer_0.Out Side Left)
   Multiply_3 = UE.Multiply:Audio(AdditionalOperands=@n1.Value, PrimaryOperand=@WavePlayer_0.Out Front Left)
   Multiply_4 = UE.Multiply:Audio(PrimaryOperand=@n1.Value, AdditionalOperands=@WavePlayer_0.Out Front Center)
   Multiply_5 = UE.Multiply:Audio(PrimaryOperand=@n1.Value, AdditionalOperands=@WavePlayer_0.Out Front Right)
  ]
  links [
   {dst="@UE.OutputFormat.5dot1.Audio:3", src=Multiply_0.Out}
   {dst="@UE.OutputFormat.5dot1.Audio:5", src=Multiply_1.Out}
   {dst="@UE.OutputFormat.5dot1.Audio:4", src=Multiply_2.Out}
   {dst="@UE.OutputFormat.5dot1.Audio:0", src=Multiply_3.Out}
   {dst="@UE.OutputFormat.5dot1.Audio:2", src=Multiply_4.Out}
   {dst="@UE.OutputFormat.5dot1.Audio:1", src=Multiply_5.Out}
  ]
 }
 Mobile {
  nodes [
   InitVariable.Trigger()
   InitVariable.Trigger()
   InitVariable.Trigger()
   InitVariable.Trigger()
   ADSREnvelopeADSREnvelope_0 = ADSR Envelope.ADSR Envelope:Audio(Trigger Attack=^UE.Source.OnPlay) {also {defaults {Attack Time=1, Decay Time=3, Sustain Level=0.8}}}
   n1 = UE.Reroute(Value=UE.Reroute(Value=@ADSREnvelopeADSREnvelope_0.Out Envelope).Value)
   WavePlayer_0 = UE.Wave Player:Mono(Wave Asset=^Amb_Loop_Mono, Play=@ADSREnvelopeADSREnvelope_0.On Attack Triggered) {also {defaults {Loop="<MetaSoundBuilderResult.SUCCEEDED: 0>"}, version [1,1]}}
   Multiply_0 = UE.Multiply:Audio(PrimaryOperand=@n1.Value, AdditionalOperands=@WavePlayer_0.Out Mono)
   Multiply_1 = UE.Multiply:Audio(PrimaryOperand=@n1.Value, AdditionalOperands=@WavePlayer_0.Out Mono)
  ]
  links [{dst="@UE.OutputFormat.5dot1.Audio:0", src=Multiply_0.Out}, {dst="@UE.OutputFormat.5dot1.Audio:1", src=Multiply_1.Out}]
 }
]
properties {
 assetImportData=&self:AssetImportData
 attenuationSettings=&/ATT_BasilOkra_ConsoleAmbience
 bLooping=true
 duration=10000
 maxDistance=580
 numChannels=6
 outputFormat=FiveDotOne
 sampleRate=48000
 soundClassObject=/Game/Sounds/FortSoundClassesAndModes/Fort_LTM_BR_InGame_Ambience
 modulationSettings {volumeRouting=Override, volumeModulationDestination {bEnableModulation=true, modulators [/WildEstate_BasilOkra/Assets/Audio/CB_BasilOkra]}}
}
