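# Stereo MetaSound source for the BasilOkra console gas leak, a spatialized looping ambient effect built on a gas leak wave asset.
# On play, a stereo wave player feeds left and right through separate multiplies sharing one Multiplier input that trims gain.
# It borrows Fortnite chest attenuation and the loot ambient loop sound class.
# It keeps playing when silent, overrides concurrency to cap voices, and routes volume through the BasilOkra modulator.
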MetaSoundSource /WildEstate_BasilOkra/Assets/Audio/SFX/MSS_BasilOkra_Console_GasLeak_LP
inputs [
 Gas Leak Loop {type=WaveAsset}
 Multiplier {type=Float}
 UE.Source.OnPlay {type=Trigger}
]
interfaces [UE.OutputFormat.Stereo:1.0, UE.Source:1.1]
outputs [
 "UE.OutputFormat.Stereo.Audio:0" {type=Audio}
 "UE.OutputFormat.Stereo.Audio:1" {type=Audio}
]
pages [Default {
  input_defaults {Gas Leak Loop=/WildEstate_BasilOkra/Assets/Audio/SFX/wav/BasilOkra_GasLeak_Loop, Multiplier=0.7, UE.Source.OnPlay=null}
  nodes [
   WavePlayer_0 = UE.Wave Player:Stereo(Wave Asset=^Gas Leak Loop, Play=^UE.Source.OnPlay) {also {defaults {Loop="<MetaSoundBuilderResult.SUCCEEDED: 0>"}, version [1,1]}}
   Multiply_0 = UE.Multiply:Audio by Float(AdditionalOperands=^Multiplier, PrimaryOperand=@WavePlayer_0.Out Left)
   Multiply_1 = UE.Multiply:Audio by Float(AdditionalOperands=^Multiplier, PrimaryOperand=@WavePlayer_0.Out Right)
  ]
  links [{dst="@UE.OutputFormat.Stereo.Audio:0", src=Multiply_0.Out}, {dst="@UE.OutputFormat.Stereo.Audio:1", src=Multiply_1.Out}]
 }]
properties {
 assetImportData=&self:AssetImportData
 attenuationSettings=/Game/Sounds/Attenuation/Fort_Chests_Att
 bLooping=true
 bOverrideConcurrency=true
 duration=10000
 maxDistance=1300
 numChannels=2
 outputFormat=Stereo
 sampleRate=48000
 soundClassObject=/Game/Sounds/FortSoundClassesAndModes/Fort_Loot_Ambient_Loop
 virtualizationMode=PlayWhenSilent
 concurrencyOverrides {bVolumeScaleCanRelease=true, maxCount=2, retriggerTime=0.1, voiceStealReleaseTime=0.2, volumeScaleAttackTime=0.2, volumeScaleReleaseTime=0.2}
 concurrencySet [&/CONC_BasilOkra_Alarms]
 modulationSettings {volumeRouting=Override, volumeModulationDestination {bEnableModulation=true, modulators [/WildEstate_BasilOkra/Assets/Audio/CB_BasilOkra]}}
}
