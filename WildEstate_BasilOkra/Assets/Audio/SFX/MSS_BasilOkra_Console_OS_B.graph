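# Mono MetaSound that plays intermittent ambient one-shots for the BasilOkra console, using set B of wave assets.
# On play, a random delay of a few seconds precedes a random wave pick, and each finished wave retriggers the cycle through trigger variables.
# Playback start resets the random timer, and OnStop halts the player and signals OnFinished.
# Output is gain-boosted by Multiplier, with console attenuation and control bus volume modulation.
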MetaSoundSource /WildEstate_BasilOkra/Assets/Audio/SFX/MSS_BasilOkra_Console_OS_B
inputs [
 Multiplier {type=Float}
 OneShots B {type=WaveAsset:Array}
 OnStop {type=Trigger}
 UE.Source.OnPlay {type=Trigger}
]
interfaces [UE.OutputFormat.Mono:1.0, UE.Source.OneShot:1.0, UE.Source:1.1]
outputs [
 "UE.OutputFormat.Mono.Audio:0" {type=Audio}
 UE.Source.OneShot.OnFinished {type=Trigger}
]
pages [Default {
  input_defaults {
   Multiplier=1.6
   OnStop=null
   UE.Source.OnPlay=null
   OneShots B [
    /WildEstate_BasilOkra/Assets/Audio/SFX/wav/BasilOkra_Amb_OS_07
    /WildEstate_BasilOkra/Assets/Audio/SFX/wav/BasilOkra_Amb_OS_08
    /WildEstate_BasilOkra/Assets/Audio/SFX/wav/BasilOkra_Amb_OS_09
    /WildEstate_BasilOkra/Assets/Audio/SFX/wav/BasilOkra_Amb_OS_10
    /WildEstate_BasilOkra/Assets/Audio/SFX/wav/BasilOkra_Amb_OS_11
    /WildEstate_BasilOkra/Assets/Audio/SFX/wav/BasilOkra_Amb_OS_12
   ]
  }
  nodes [
   n1 = VariableDeferredAccessor.Trigger(Variable=InitVariable.Trigger().Variable)
   TriggerAnyTriggerAny2_0 = "TriggerAny.Trigger Any (2)"(In 0=^UE.Source.OnPlay, In 1=@n1.Value)
   n2 = VariableDeferredAccessor.Trigger(Variable=InitVariable.Trigger().Variable)
   n3 = UE.RandomTime(Next=@TriggerAnyTriggerAny2_0.Out, Reset=@n2.Value) {also {defaults {Max=7, Min=3}}}
   TriggerDelay_0 = UE.Trigger Delay(In=@n3.On Next, Delay Time=@n3.Value)
   ArrayRandomGet_0 = Array.Random Get:WaveAsset:Array(In Array=^OneShots B, Next=@TriggerDelay_0.Out) {also {defaults {Weights []}}}
   WavePlayer_0 = UE.Wave Player:Mono(Stop=^OnStop, Play=@ArrayRandomGet_0.On Next, Wave Asset=@ArrayRandomGet_0.Value) {also {version [1,1]}}
   Multiply_0 = UE.Multiply:Audio by Float(AdditionalOperands=^Multiplier, PrimaryOperand=@WavePlayer_0.Out Mono)
   VariableMutator.Trigger(Variable=@n2.Variable, Value=@WavePlayer_0.On Play)
   VariableMutator.Trigger(Variable=@n1.Variable, Value=@WavePlayer_0.On Finished)
  ]
  links [{dst="@UE.Source.OneShot.OnFinished", src="@OnStop"}, {dst="@UE.OutputFormat.Mono.Audio:0", src=Multiply_0.Out}]
 }]
properties {
 assetImportData=&self:AssetImportData
 attenuationSettings=&/ATT_BasilOkra_Console_OS
 maxDistance=887.7905
 sampleRate=48000
 soundClassObject=/Game/Sounds/FortSoundClassesAndModes/Fort_LTM_BR_InGame_Ambience
 modulationSettings {volumeRouting=Override, volumeModulationDestination {bEnableModulation=true, modulators [/WildEstate_BasilOkra/Assets/Audio/CB_BasilOkra]}}
}
