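# One-shot mono MetaSound for a TV interaction-complete effect, spatialized with the TV monitor attenuation.
# On play, a randomly picked flicker wave and the fixed TV A wave play layered, both pitch-shifted by one shared random value.
# A two-input mixer boosts and sums the layers, and the source finishes once both players end; the Multiplier input is unconnected.
# Volume modulation is overridden with the MeltdownDuck and TVs control buses.
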MetaSoundSource /WildEstate_BasilOkra/Assets/Audio/SFX/MSS_BasilOkra_Interact_Complete_TV_02
inputs [
 Flicker {type=WaveAsset:Array}
 Multiplier {type=Float}
 TV A {type=WaveAsset}
 UE.Source.OnPlay {type=Trigger}
]
interfaces [UE.OutputFormat.Mono:1.0, UE.Source.OneShot:1.0, UE.Source:1.1]
outputs [
 "UE.OutputFormat.Mono.Audio:0" {type=Audio}
 UE.Source.OneShot.OnFinished {type=Trigger}
]
pages [Default {
  input_defaults {
   Multiplier=3
   TV A=/WildEstate_BasilOkra/Assets/Audio/SFX/wav/BasilOkra_Interact_Complete_TV_A
   UE.Source.OnPlay=null
   Flicker [
    /WildEstate_BasilOkra/Assets/Audio/SFX/wav/BasilOkra_Interact_Complete_TV_Flicker_OS_04
    /WildEstate_BasilOkra/Assets/Audio/SFX/wav/BasilOkra_Interact_Complete_TV_Flicker_OS_05
    /WildEstate_BasilOkra/Assets/Audio/SFX/wav/BasilOkra_Interact_Complete_TV_Flicker_OS_06
   ]
  }
  nodes [
   ArrayRandomGet_0 = Array.Random Get:WaveAsset:Array(In Array=^Flicker, Next=^UE.Source.OnPlay) {also {defaults {Weights []}}}
   n1 = UE.RandomFloat(Next=^UE.Source.OnPlay) {also {defaults {Min=-3}}}
   WavePlayer_0 = UE.Wave Player:Mono(Play=@ArrayRandomGet_0.On Next, Wave Asset=@ArrayRandomGet_0.Value, Pitch Shift=@n1.Value) {also {version [1,1]}}
   WavePlayer_1 = UE.Wave Player:Mono(Wave Asset=^TV A, Play=@n1.On Next, Pitch Shift=@n1.Value) {also {version [1,1]}}
   AudioMixerAudioMixerMono2_0 = "AudioMixer.Audio Mixer (Mono, 2)"(In 1=@WavePlayer_0.Out Mono, In 0=@WavePlayer_1.Out Mono) {also {defaults {Gain 0=2.5, Gain 1=2.5}}}
   TriggerAccumulatorTriggerAccumulate2_0 = "TriggerAccumulator.Trigger Accumulate (2)"(In 1=@WavePlayer_0.On Finished, In 0=@WavePlayer_1.On Finished)
  ]
  links [
   {dst="@UE.OutputFormat.Mono.Audio:0", src=AudioMixerAudioMixerMono2_0.Out}
   {dst="@UE.Source.OneShot.OnFinished", src=TriggerAccumulatorTriggerAccumulate2_0.Out}
  ]
 }]
properties {
 assetImportData=&self:AssetImportData
 attenuationSettings=&/ATT_BasilOkra_TV_Monitors
 maxDistance=465
 sampleRate=48000
 soundClassObject=&Fort_SFX_InGame
 modulationSettings {
  volumeRouting=Override
  volumeModulationDestination {
   bEnableModulation=true
   modulators [/WildEstate_BasilOkra/Assets/Audio/CB_BasilOkra_MeltdownDuck, /WildEstate_BasilOkra/Assets/Audio/CB_BasilOkra_TVs]
  }
 }
}
