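# One-shot stereo MetaSound source for the BasilOkra interact highlight sound effect, using console-interact attenuation.
# A Wave Player takes the Press wave asset, and a Multiplier input scales both channels before output.
# OnPlay links straight to OnFinished, and no trigger is wired to the Wave Player's Play input.
# Highlight concurrency limits overlapping instances, and a bus send feeds the music-volume sidechain.
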MetaSoundSource /WildEstate_BasilOkra/Assets/Audio/SFX/MSS_BasilOkra_Interact_Highlight_01
inputs [
 Multiplier {type=Float}
 Press {type=WaveAsset}
 UE.Source.OnPlay {type=Trigger}
]
interfaces [UE.OutputFormat.Stereo:1.0, UE.Source.OneShot:1.0, UE.Source:1.1]
outputs [
 "UE.OutputFormat.Stereo.Audio:0" {type=Audio}
 "UE.OutputFormat.Stereo.Audio:1" {type=Audio}
 UE.Source.OneShot.OnFinished {type=Trigger}
]
pages [Default {
  input_defaults {Multiplier=1, Press=/WildEstate_BasilOkra/Assets/Audio/SFX/wav/BasilOkra_Interact_Highlight_01, UE.Source.OnPlay=null}
  nodes [
   WavePlayer_0 = UE.Wave Player:Stereo(Wave Asset=^Press) {also {version [1,1]}}
   Multiply_0 = UE.Multiply:Audio by Float(AdditionalOperands=^Multiplier, PrimaryOperand=@WavePlayer_0.Out Right)
   Multiply_1 = UE.Multiply:Audio by Float(AdditionalOperands=^Multiplier, PrimaryOperand=@WavePlayer_0.Out Left)
  ]
  links [
   {dst="@UE.Source.OneShot.OnFinished", src="@UE.Source.OnPlay"}
   {dst="@UE.OutputFormat.Stereo.Audio:1", src=Multiply_0.Out}
   {dst="@UE.OutputFormat.Stereo.Audio:0", src=Multiply_1.Out}
  ]
 }]
properties {
 assetImportData=&self:AssetImportData
 attenuationSettings=&/ATT_BasilOkra_Console_Interact
 maxDistance=2000
 numChannels=2
 outputFormat=Stereo
 sampleRate=48000
 soundClassObject=&Fort_SFX_InGame
 busSends [{audioBus=&AB_Sidechain_FortMusic_Vol, sendLevel=0.65}]
 concurrencyOverrides {bVolumeScaleCanRelease=true, maxCount=3, retriggerTime=1}
 concurrencySet [&/CONC_BasilOkra_Highlights]
}
