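# One-shot stereo MetaSound source for the BasilOkra meltdown shutter sound effect.
# On play, it picks a random wave from five shutter variations, plays it through a stereo wave player and scales both channels by a shared gain multiplier.
# The source finishes with the wave.
# It uses meltdown attenuation, overrides concurrency with a voice cap and retrigger time, and routes volume modulation to the BasilOkra and Chaos control buses.
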MetaSoundSource /WildEstate_BasilOkra/Assets/Audio/SFX/MSS_BasilOkra_Meltdown_Shutter_OS
inputs [
 Meltdown OneShots {type=WaveAsset:Array}
 Multiplier {type=Float}
 UE.Source.OnPlay {type=Trigger}
]
interfaces [UE.OutputFormat.Stereo:1.0, UE.Source.OneShot:1.0, UE.Source:1.1]
outputs [
 "UE.OutputFormat.Stereo.Audio:0" {type=Audio}
 "UE.OutputFormat.Stereo.Audio:1" {type=Audio}
 UE.Source.OneShot.OnFinished {type=Trigger}
]
pages [Default {
  input_defaults {
   Multiplier=0.8
   UE.Source.OnPlay=null
   Meltdown OneShots [
    /WildEstate_BasilOkra/Assets/Audio/SFX/wav/BasilOkra_Meltdown_Shutter_OS_01
    /WildEstate_BasilOkra/Assets/Audio/SFX/wav/BasilOkra_Meltdown_Shutter_OS_02
    /WildEstate_BasilOkra/Assets/Audio/SFX/wav/BasilOkra_Meltdown_Shutter_OS_03
    /WildEstate_BasilOkra/Assets/Audio/SFX/wav/BasilOkra_Meltdown_Shutter_OS_04
    /WildEstate_BasilOkra/Assets/Audio/SFX/wav/BasilOkra_Meltdown_Shutter_OS_05
   ]
  }
  nodes [
   ArrayRandomGet_0 = Array.Random Get:WaveAsset:Array(In Array=^Meltdown OneShots, Next=^UE.Source.OnPlay)
   WavePlayer_0 = UE.Wave Player:Stereo(Play=@ArrayRandomGet_0.On Next, Wave Asset=@ArrayRandomGet_0.Value) {also {version [1,1]}}
   Multiply_0 = UE.Multiply:Audio by Float(AdditionalOperands=^Multiplier, PrimaryOperand=@WavePlayer_0.Out Right)
   Multiply_1 = UE.Multiply:Audio by Float(AdditionalOperands=^Multiplier, PrimaryOperand=@WavePlayer_0.Out Left)
  ]
  links [
   {dst="@UE.OutputFormat.Stereo.Audio:1", src=Multiply_0.Out}
   {dst="@UE.OutputFormat.Stereo.Audio:0", src=Multiply_1.Out}
   {dst="@UE.Source.OneShot.OnFinished", src=WavePlayer_0.On Finished}
  ]
 }]
properties {
 assetImportData=&self:AssetImportData
 attenuationSettings=&/ATT_BasilOkra_Meltdown_OS
 bOverrideConcurrency=true
 maxDistance=6000
 numChannels=2
 outputFormat=Stereo
 sampleRate=48000
 soundClassObject=&Fort_SFX_InGame
 concurrencyOverrides {maxCount=4, retriggerTime=3, volumeScale=0.6}
 modulationSettings {
  volumeRouting=Override
  volumeModulationDestination {bEnableModulation=true, modulators [/WildEstate_BasilOkra/Assets/Audio/CB_BasilOkra, /WildEstate_BasilOkra/Assets/Audio/CB_BasilOkra_Chaos]}
 }
}
