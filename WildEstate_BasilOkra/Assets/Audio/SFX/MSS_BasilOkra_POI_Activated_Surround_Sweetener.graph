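# One-shot 5.1 MetaSound source for the surround sweetener layer of the BasilOkra POI activation SFX, attenuated via its own attenuation asset.
# On play, a 5.1 Wave Player plays the surround wave, each channel scaled by a shared Multiplier float, and signals OnFinished when playback ends.
# A separate Mobile page swaps in a mono wave and feeds only front left and right through the same gain.
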MetaSoundSource /WildEstate_BasilOkra/Assets/Audio/SFX/MSS_BasilOkra_POI_Activated_Surround_Sweetener
inputs [
 Multiplier {type=Float}
 POI Activation {type=WaveAsset}
 POI Activation _Mobile {type=WaveAsset}
 UE.Source.OnPlay {type=Trigger}
]
interfaces [UE.OutputFormat.5dot1:1.0, UE.Source.OneShot:1.0, UE.Source:1.1]
outputs [
 "UE.OutputFormat.5dot1.Audio:0" {type=Audio}
 "UE.OutputFormat.5dot1.Audio:1" {type=Audio}
 "UE.OutputFormat.5dot1.Audio:2" {type=Audio}
 "UE.OutputFormat.5dot1.Audio:3" {type=Audio}
 "UE.OutputFormat.5dot1.Audio:4" {type=Audio}
 "UE.OutputFormat.5dot1.Audio:5" {type=Audio}
 UE.Source.OneShot.OnFinished {type=Trigger}
]
pages [
 Default {
  is_build_graph=true
  input_defaults {
   Multiplier=1
   POI Activation=/WildEstate_BasilOkra/Assets/Audio/SFX/wav/BasilOkra_POI_Activate_Surround
   POI Activation _Mobile=/WildEstate_BasilOkra/Assets/Audio/SFX/wav/BasilOkra_POI_Activate_Mono
   UE.Source.OnPlay=null
  }
  nodes [
   WavePlayer_0 = UE.Wave Player:5dot1(Wave Asset=^POI Activation, Play=^UE.Source.OnPlay) {also {version [1,1]}}
   Multiply_0 = UE.Multiply:Audio by Float(AdditionalOperands=^Multiplier, PrimaryOperand=@WavePlayer_0.Out Low Frequency)
   Multiply_1 = UE.Multiply:Audio by Float(AdditionalOperands=^Multiplier, PrimaryOperand=@WavePlayer_0.Out Front Right)
   Multiply_2 = UE.Multiply:Audio by Float(AdditionalOperands=^Multiplier, PrimaryOperand=@WavePlayer_0.Out Side Right)
   Multiply_3 = UE.Multiply:Audio by Float(AdditionalOperands=^Multiplier, PrimaryOperand=@WavePlayer_0.Out Side Left)
   Multiply_4 = UE.Multiply:Audio by Float(AdditionalOperands=^Multiplier, PrimaryOperand=@WavePlayer_0.Out Front Center)
   Multiply_5 = UE.Multiply:Audio by Float(AdditionalOperands=^Multiplier, PrimaryOperand=@WavePlayer_0.Out Front Left)
  ]
  links [
   {dst="@UE.OutputFormat.5dot1.Audio:3", src=Multiply_0.Out}
   {dst="@UE.OutputFormat.5dot1.Audio:1", src=Multiply_1.Out}
   {dst="@UE.OutputFormat.5dot1.Audio:5", src=Multiply_2.Out}
   {dst="@UE.OutputFormat.5dot1.Audio:4", src=Multiply_3.Out}
   {dst="@UE.OutputFormat.5dot1.Audio:2", src=Multiply_4.Out}
   {dst="@UE.OutputFormat.5dot1.Audio:0", src=Multiply_5.Out}
   {dst="@UE.Source.OneShot.OnFinished", src=WavePlayer_0.On Finished}
  ]
 }
 Mobile {
  nodes [
   WavePlayer_0 = UE.Wave Player:5dot1(Wave Asset=^POI Activation _Mobile, Play=^UE.Source.OnPlay) {also {version [1,1]}}
   Multiply_0 = UE.Multiply:Audio by Float(AdditionalOperands=^Multiplier, PrimaryOperand=@WavePlayer_0.Out Front Right)
   Multiply_1 = UE.Multiply:Audio by Float(AdditionalOperands=^Multiplier, PrimaryOperand=@WavePlayer_0.Out Front Left)
  ]
  links [
   {dst="@UE.OutputFormat.5dot1.Audio:1", src=Multiply_0.Out}
   {dst="@UE.OutputFormat.5dot1.Audio:0", src=Multiply_1.Out}
   {dst="@UE.Source.OneShot.OnFinished", src=WavePlayer_0.On Finished}
  ]
 }
]
properties {
 assetImportData=&self:AssetImportData
 attenuationSettings=&/ATT_BasilOkra_POI_Activated_Surround
 maxDistance=20000
 numChannels=6
 outputFormat=FiveDotOne
 sampleRate=48000
 soundClassObject=&Fort_SFX_InGame
}
