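# One-shot stereo MetaSound Source for the player-launched-by-vent sound effect in BasilOkra.
# On play, a random get picks one of two launch wave variations and feeds a stereo wave player, whose finish ends the source.
# The Succeeded input holds a puzzle success wave but connects to nothing in the graph.
# It uses its own attenuation settings and the in-game SFX sound class.
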MetaSoundSource /WildEstate_BasilOkra/Assets/Audio/SFX/MSS_BasilOkra_PlayerLaunchedByVent
inputs [
 Player Launched {type=WaveAsset:Array}
 Succeeded {type=WaveAsset}
 UE.Source.OnPlay {type=Trigger}
]
interfaces [UE.OutputFormat.Stereo:1.0, UE.Source.OneShot:1.0, UE.Source:1.1]
outputs [
 "UE.OutputFormat.Stereo.Audio:0" {type=Audio}
 "UE.OutputFormat.Stereo.Audio:1" {type=Audio}
 UE.Source.OneShot.OnFinished {type=Trigger}
]
pages [Default {
  input_defaults {
   Succeeded=/WildEstate_BasilOkra/Assets/Audio/SFX/wav/BasilOkra_Puzzle_Success
   UE.Source.OnPlay=null
   Player Launched [
    /WildEstate_BasilOkra/Assets/Audio/SFX/wav/BasilOkra_PlayerLaunched_0S_01
    /WildEstate_BasilOkra/Assets/Audio/SFX/wav/BasilOkra_PlayerLaunched_0S_02
   ]
  }
  nodes [
   ArrayRandomGet_0 = Array.Random Get:WaveAsset:Array(In Array=^Player Launched, Next=^UE.Source.OnPlay)
   WavePlayer_0 = UE.Wave Player:Stereo(Play=@ArrayRandomGet_0.On Next, Wave Asset=@ArrayRandomGet_0.Value) {also {version [1,1]}}
  ]
  links [
   {dst="@UE.Source.OneShot.OnFinished", src=WavePlayer_0.On Finished}
   {dst="@UE.OutputFormat.Stereo.Audio:0", src=WavePlayer_0.Out Left}
   {dst="@UE.OutputFormat.Stereo.Audio:1", src=WavePlayer_0.Out Right}
  ]
 }]
properties {
 assetImportData=&self:AssetImportData
 attenuationSettings=&/ATT_BasilOkra_PlayerLaunchedByVent
 maxDistance=5500
 numChannels=2
 outputFormat=Stereo
 sampleRate=48000
 soundClassObject=&Fort_SFX_InGame
}
