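# One-shot stereo MetaSound stinger for a failed BasilOkra puzzle, spatialized by puzzle stinger attenuation.
# Default page mixes a random fail clip with a meltdown alarm delayed 0.2 seconds, finishing when both players end.
# Mobile page plays only the fail clip, feeding a four-input finish accumulator with one input wired.
# Shutter and Vents arrays have defaults, trimmed on Mobile, but neither page's nodes use them.
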MetaSoundSource /WildEstate_BasilOkra/Assets/Audio/SFX/MSS_BasilOkra_Puzzle_Failed
inputs [
 Failed {type=WaveAsset:Array}
 MeltdownAlarm {type=WaveAsset}
 Shutter {type=WaveAsset:Array}
 UE.Source.OnPlay {type=Trigger}
 Vents {type=WaveAsset:Array}
]
interfaces [UE.OutputFormat.Stereo:1.0, UE.Source.OneShot:1.0, UE.Source:1.1]
outputs [
 "UE.OutputFormat.Stereo.Audio:0" {type=Audio}
 "UE.OutputFormat.Stereo.Audio:1" {type=Audio}
 UE.Source.OneShot.OnFinished {type=Trigger}
]
pages [
 Default {
  is_build_graph=true
  input_defaults {
   MeltdownAlarm=/WildEstate_BasilOkra/Assets/Audio/SFX/wav/BasilOkra_IP_Interact_Fail_SingleAlarm
   UE.Source.OnPlay=null
   Failed [
    /WildEstate_BasilOkra/Assets/Audio/SFX/wav/BasilOkra_IP_Puzzle_Fail_OS_01
    /WildEstate_BasilOkra/Assets/Audio/SFX/wav/BasilOkra_IP_Puzzle_Fail_OS_02
    /WildEstate_BasilOkra/Assets/Audio/SFX/wav/BasilOkra_IP_Puzzle_Fail_OS_03
   ]
   Shutter [
    /WildEstate_BasilOkra/Assets/Audio/SFX/wav/BasilOkra_Meltdown_Shutter_OS_01
    /WildEstate_BasilOkra/Assets/Audio/SFX/wav/BasilOkra_Meltdown_Shutter_OS_02
    /WildEstate_BasilOkra/Assets/Audio/SFX/wav/BasilOkra_Meltdown_Shutter_OS_03
    /WildEstate_BasilOkra/Assets/Audio/SFX/wav/BasilOkra_Meltdown_Shutter_OS_04
    /WildEstate_BasilOkra/Assets/Audio/SFX/wav/BasilOkra_Meltdown_Shutter_OS_05
   ]
   Vents [
    /WildEstate_BasilOkra/Assets/Audio/SFX/wav/BasilOkra_Meltdown_Vent_OS_01
    /WildEstate_BasilOkra/Assets/Audio/SFX/wav/BasilOkra_Meltdown_Vent_OS_02
    /WildEstate_BasilOkra/Assets/Audio/SFX/wav/BasilOkra_Meltdown_Vent_OS_03
    /WildEstate_BasilOkra/Assets/Audio/SFX/wav/BasilOkra_Meltdown_Vent_OS_04
    /WildEstate_BasilOkra/Assets/Audio/SFX/wav/BasilOkra_Meltdown_Vent_OS_05
    /WildEstate_BasilOkra/Assets/Audio/SFX/wav/BasilOkra_Meltdown_Vent_OS_06
   ]
  }
  nodes [
   TriggerDelay_0 = UE.Trigger Delay(In=^UE.Source.OnPlay) {also {defaults {Delay Time=0.2}}}
   WavePlayer_0 = UE.Wave Player:Stereo(Wave Asset=^MeltdownAlarm, Play=@TriggerDelay_0.Out) {also {version [1,1]}}
   ArrayRandomGet_0 = Array.Random Get:WaveAsset:Array(In Array=^Failed, Next=^UE.Source.OnPlay) {also {defaults {Weights []}}}
   WavePlayer_1 = UE.Wave Player:Stereo(Play=@ArrayRandomGet_0.On Next, Wave Asset=@ArrayRandomGet_0.Value) {also {version [1,1]}}
   AudioMixerAudioMixerStereo2_0 = "AudioMixer.Audio Mixer (Stereo, 2)"(In 1 L=@WavePlayer_0.Out Left, In 1 R=@WavePlayer_0.Out Right, In 0 L=@WavePlayer_1.Out Left, In 0 R=@WavePlayer_1.Out Right) {also {defaults {Gain 0=0.85, Gain 1=1.1}}}
   TriggerAccumulatorTriggerAccumulate2_0 = "TriggerAccumulator.Trigger Accumulate (2)"(In 1=@WavePlayer_0.On Finished, In 0=@WavePlayer_1.On Finished)
  ]
  links [
   {dst="@UE.OutputFormat.Stereo.Audio:0", src=AudioMixerAudioMixerStereo2_0.Out L}
   {dst="@UE.OutputFormat.Stereo.Audio:1", src=AudioMixerAudioMixerStereo2_0.Out R}
   {dst="@UE.Source.OneShot.OnFinished", src=TriggerAccumulatorTriggerAccumulate2_0.Out}
  ]
 }
 Mobile {
  input_defaults {
   Shutter [
    &BasilOkra_Meltdown_Shutter_OS_01
    &BasilOkra_Meltdown_Shutter_OS_02
    &BasilOkra_Meltdown_Shutter_OS_03
   ]
   Vents [
    &BasilOkra_Meltdown_Vent_OS_01
    &BasilOkra_Meltdown_Vent_OS_02
    &BasilOkra_Meltdown_Vent_OS_03
   ]
  }
  nodes [
   ArrayRandomGet_0 = Array.Random Get:WaveAsset:Array(In Array=^Failed, Next=^UE.Source.OnPlay) {also {defaults {Weights []}}}
   WavePlayer_0 = UE.Wave Player:Stereo(Play=@ArrayRandomGet_0.On Next, Wave Asset=@ArrayRandomGet_0.Value) {also {version [1,1]}}
   TriggerAccumulatorTriggerAccumulate4_0 = "TriggerAccumulator.Trigger Accumulate (4)"(In 0=@WavePlayer_0.On Finished)
  ]
  links [
   {dst="@UE.Source.OneShot.OnFinished", src=TriggerAccumulatorTriggerAccumulate4_0.Out}
   {dst="@UE.OutputFormat.Stereo.Audio:0", src=WavePlayer_0.Out Left}
   {dst="@UE.OutputFormat.Stereo.Audio:1", src=WavePlayer_0.Out Right}
  ]
 }
]
properties {
 assetImportData=&self:AssetImportData
 attenuationSettings=&/ATT_BasilOkra_Puzzle_Stingers
 maxDistance=3000
 numChannels=2
 outputFormat=Stereo
 sampleRate=48000
 soundClassObject=&Fort_SFX_InGame
}
