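# A one-shot 5.1 MetaSound source that plays the success sound for the Basil Okra shutdown sequence puzzle.
# On play, it fires a stereo bell ring and a 5.1 surround sweetener together.
# The bell is boosted and mixed into the front left and right with the sweetener's fronts, while a shared Multiplier scales every sweetener channel.
# It finishes only after both wave players end, via a trigger accumulator.
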MetaSoundSource /WildEstate_BasilOkra/Assets/Audio/SFX/MSS_BasilOkra_Puzzle_Shutdown_Sequence_Succeed
inputs [
 Bell Ring {type=WaveAsset}
 Multiplier {type=Float}
 Shutdown Surround {type=WaveAsset}
 UE.Source.OnPlay {type=Trigger}
]
interfaces [UE.OutputFormat.5dot1:1.0, UE.Source.OneShot:1.0, UE.Source:1.1]
outputs [
 "UE.OutputFormat.5dot1.Audio:0" {type=Audio}
 "UE.OutputFormat.5dot1.Audio:1" {type=Audio}
 "UE.OutputFormat.5dot1.Audio:2" {type=Audio}
 "UE.OutputFormat.5dot1.Audio:3" {type=Audio}
 "UE.OutputFormat.5dot1.Audio:4" {type=Audio}
 "UE.OutputFormat.5dot1.Audio:5" {type=Audio}
 UE.Source.OneShot.OnFinished {type=Trigger}
]
pages [Default {
  input_defaults {
   Bell Ring=/WildEstate_BasilOkra/Assets/Audio/SFX/wav/BasilOkra_Puzzle_Succeed_BellRIng
   Multiplier=0.7
   Shutdown Surround=/WildEstate_BasilOkra/Assets/Audio/SFX/wav/BasilOkra_IP_Puzzle_Succeed_5_1_Sweetener
   UE.Source.OnPlay=null
  }
  nodes [
   WavePlayer_0 = UE.Wave Player:5dot1(Wave Asset=^Shutdown Surround, Play=^UE.Source.OnPlay) {also {version [1,1]}}
   WavePlayer_1 = UE.Wave Player:Stereo(Wave Asset=^Bell Ring, Play=^UE.Source.OnPlay) {also {version [1,1]}}
   AudioMixerAudioMixerStereo2_0 = "AudioMixer.Audio Mixer (Stereo, 2)"(
    Gain 1=^Multiplier
    In 1 L=@WavePlayer_0.Out Front Left
    In 1 R=@WavePlayer_0.Out Front Right
    In 0 L=@WavePlayer_1.Out Left
    In 0 R=@WavePlayer_1.Out Right
   ) {also {defaults {Gain 0=1.8}}}
   Multiply_0 = UE.Multiply:Audio by Float(AdditionalOperands=^Multiplier, PrimaryOperand=@WavePlayer_0.Out Low Frequency)
   Multiply_1 = UE.Multiply:Audio by Float(AdditionalOperands=^Multiplier, PrimaryOperand=@WavePlayer_0.Out Side Right)
   Multiply_2 = UE.Multiply:Audio by Float(AdditionalOperands=^Multiplier, PrimaryOperand=@WavePlayer_0.Out Side Left)
   Multiply_3 = UE.Multiply:Audio by Float(AdditionalOperands=^Multiplier, PrimaryOperand=@WavePlayer_0.Out Front Center)
   TriggerAccumulatorTriggerAccumulate2_0 = "TriggerAccumulator.Trigger Accumulate (2)"(In 1=@WavePlayer_0.On Finished, In 0=@WavePlayer_1.On Finished)
  ]
  links [
   {dst="@UE.OutputFormat.5dot1.Audio:0", src=AudioMixerAudioMixerStereo2_0.Out L}
   {dst="@UE.OutputFormat.5dot1.Audio:1", src=AudioMixerAudioMixerStereo2_0.Out R}
   {dst="@UE.OutputFormat.5dot1.Audio:3", src=Multiply_0.Out}
   {dst="@UE.OutputFormat.5dot1.Audio:5", src=Multiply_1.Out}
   {dst="@UE.OutputFormat.5dot1.Audio:4", src=Multiply_2.Out}
   {dst="@UE.OutputFormat.5dot1.Audio:2", src=Multiply_3.Out}
   {dst="@UE.Source.OneShot.OnFinished", src=TriggerAccumulatorTriggerAccumulate2_0.Out}
  ]
 }]
properties {
 assetImportData=&self:AssetImportData
 attenuationSettings=&/ATT_BasilOkra_Puzzle_Sweeteners
 maxDistance=6000
 numChannels=6
 outputFormat=FiveDotOne
 sampleRate=48000
 soundClassObject=&Fort_SFX_InGame
}
