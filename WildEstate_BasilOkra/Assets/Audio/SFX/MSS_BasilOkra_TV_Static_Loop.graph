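# Mono MetaSound source for spatialized TV static ambience, using a TV monitor attenuation asset and the in-game ambience sound class.
# OnPlay triggers an ADSR envelope, whose attack picks one of two static waves at random for the wave player.
# The wave is multiplied by the envelope and the Multiplier gain input, fading in on play and out when OnStop fires.
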MetaSoundSource /WildEstate_BasilOkra/Assets/Audio/SFX/MSS_BasilOkra_TV_Static_Loop
inputs [
 Multiplier {type=Float}
 OnStop {type=Trigger}
 "Static " {type=WaveAsset:Array}
 UE.Source.OnPlay {type=Trigger}
]
interfaces [UE.OutputFormat.Mono:1.0, UE.Source:1.1]
outputs ["UE.OutputFormat.Mono.Audio:0" {type=Audio}]
pages [Default {
  input_defaults {
   Multiplier=1.5
   OnStop=null
   UE.Source.OnPlay=null
   "Static " [
    /WildEstate_BasilOkra/Assets/Audio/SFX/wav/BasilOkra_TV_Static_Loop_01
    /WildEstate_BasilOkra/Assets/Audio/SFX/wav/BasilOkra_TV_Static_Loop_02
   ]
  }
  nodes [
   InitVariable.Trigger()
   InitVariable.Trigger()
   ADSREnvelopeADSREnvelope_0 = ADSR Envelope.ADSR Envelope:Audio(Trigger Release=^OnStop, Trigger Attack=^UE.Source.OnPlay) {also {defaults {Attack Time=1, Release Time=0.5, Sustain Level=1}}}
   ArrayRandomGet_0 = Array.Random Get:WaveAsset:Array(In Array=^"Static ", Next=@ADSREnvelopeADSREnvelope_0.On Attack Triggered)
   WavePlayer_0 = UE.Wave Player:Mono(Play=@ArrayRandomGet_0.On Next, Wave Asset=@ArrayRandomGet_0.Value) {also {defaults {Loop="<MetaSoundBuilderResult.SUCCEEDED: 0>"}, version [1,1]}}
   Multiply_1 = UE.Multiply:Audio by Float(
    AdditionalOperands=^Multiplier
    PrimaryOperand=UE.Multiply:Audio(AdditionalOperands=UE.Reroute(Value=@ADSREnvelopeADSREnvelope_0.Out Envelope).Value, PrimaryOperand=@WavePlayer_0.Out Mono).Out
   )
  ]
  links [{dst="@UE.OutputFormat.Mono.Audio:0", src=Multiply_1.Out}]
 }]
properties {
 assetImportData=&self:AssetImportData
 attenuationSettings=&/ATT_BasilOkra_TV_Monitors
 bLooping=true
 duration=10000
 maxDistance=465
 sampleRate=48000
 soundClassObject=/Game/Sounds/FortSoundClassesAndModes/Fort_LTM_BR_InGame_Ambience
}
