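# One-shot stereo MetaSound source for the BasilOkra UI widget banner appear sound effect.
# On play, a single Wave Player plays the WidgetAppear wave, exposed as a WaveAsset input, straight to the left and right outputs.
# The player's finished trigger ends the source.
# It uses a puzzle-stinger attenuation asset and the in-game SFX sound class.
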MetaSoundSource /WildEstate_BasilOkra/Assets/Audio/SFX/MSS_BasilOkra_UI_WidgetBannerAppear
inputs [
 UE.Source.OnPlay {type=Trigger}
 Widget Appear {type=WaveAsset}
]
interfaces [UE.OutputFormat.Stereo:1.0, UE.Source.OneShot:1.0, UE.Source:1.1]
outputs [
 "UE.OutputFormat.Stereo.Audio:0" {type=Audio}
 "UE.OutputFormat.Stereo.Audio:1" {type=Audio}
 UE.Source.OneShot.OnFinished {type=Trigger}
]
pages [Default {
  input_defaults {UE.Source.OnPlay=null, Widget Appear=/WildEstate_BasilOkra/Assets/Audio/SFX/wav/BasilOkra_IP_UI_WidgetAppear}
  nodes [WavePlayer_0 = UE.Wave Player:Stereo(Play=^UE.Source.OnPlay, Wave Asset=^Widget Appear) {also {version [1,1]}}]
  links [
   {dst="@UE.Source.OneShot.OnFinished", src=WavePlayer_0.On Finished}
   {dst="@UE.OutputFormat.Stereo.Audio:0", src=WavePlayer_0.Out Left}
   {dst="@UE.OutputFormat.Stereo.Audio:1", src=WavePlayer_0.Out Right}
  ]
 }]
properties {
 assetImportData=&self:AssetImportData
 attenuationSettings=&/ATT_BasilOkra_Puzzle_Stingers
 maxDistance=3000
 numChannels=2
 outputFormat=Stereo
 sampleRate=48000
 soundClassObject=&Fort_SFX_InGame
}
