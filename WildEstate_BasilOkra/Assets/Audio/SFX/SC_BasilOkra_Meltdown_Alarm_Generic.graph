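# Sound cue for the generic meltdown alarm in the WildEstate BasilOkra content, playing as an in-game SFX.
# A single wave player loops the BasilOkra_Alarm_01 wave directly into the root, with no variation or modulation nodes.
# Shared alarm attenuation and concurrency assets govern falloff and voice limits, alongside concurrency override values stored on the cue.
# Play-when-silent virtualization keeps the loop running when inaudible.
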SoundCue /WildEstate_BasilOkra/Assets/Audio/SFX/SC_BasilOkra_Meltdown_Alarm_Generic
root=WavePlayer_0
nodes [{id=WavePlayer_0, properties {bLooping=true, soundWaveAssetPtr=/WildEstate_BasilOkra/Assets/Audio/SFX/wav/BasilOkra_Alarm_01}}]
properties {
 attenuationSettings=&/ATT_BasilOkra_Alarms
 maxDistance=9200
 soundClassObject=&Fort_SFX_InGame
 virtualizationMode=PlayWhenSilent
 concurrencyOverrides {maxCount=4, retriggerTime=0.25, volumeScale=0.5}
 concurrencySet [&/CONC_BasilOkra_Alarms]
}
