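# Looping sound cue for the electric spark effect of the Basil Okra meltdown in Wild Estate.
# A single looping wave player plays one spark loop wave directly, with no mixing, randomization or modulation.
# The cue uses the meltdown attenuation asset for spatial falloff and routes to the in-game SFX sound class.
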SoundCue /WildEstate_BasilOkra/Assets/Audio/SFX/SC_BasilOkra_Meltdown_ElecSpark_LP
root=WavePlayer_1
nodes [{
  id=WavePlayer_1
  properties {bLooping=true, soundWaveAssetPtr=/WildEstate_BasilOkra/Assets/Audio/SFX/wav/BasilOkra_Meltdown_Elec_Spark_Loop_01}
 }]
properties {
 attenuationSettings=&/ATT_BasilOkra_Meltdown_OS
 maxDistance=6000
 soundClassObject=&Fort_SFX_InGame
 volumeMultiplier=1
}
