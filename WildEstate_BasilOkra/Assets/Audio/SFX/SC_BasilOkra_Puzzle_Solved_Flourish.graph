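# Sound Cue for the puzzle-solved flourish stinger in WildEstate BasilOkra.
# A Mixer root layers the solved flourish wave with a confetti success wave, weighting the confetti louder than the flourish.
# It uses the shared puzzle stinger attenuation asset and plays through the in-game SFX sound class at a slightly reduced cue volume.
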SoundCue /WildEstate_BasilOkra/Assets/Audio/SFX/SC_BasilOkra_Puzzle_Solved_Flourish
root=Mixer_0
nodes [
 {id=Mixer_0, children [WavePlayer_0,WavePlayer_1], properties {inputVolume [0.8,1.45]}}
 {id=WavePlayer_0, properties {soundWaveAssetPtr=/WildEstate_BasilOkra/Assets/Audio/SFX/wav/BasilOkra_Puzzle_Solved_Flourish}}
 {id=WavePlayer_1, properties {soundWaveAssetPtr=/WildEstate_BasilOkra/Assets/Audio/SFX/wav/BasilOkra_Puzzle_Succeed_Confetti}}
]
properties {
 attenuationSettings=&/ATT_BasilOkra_Puzzle_Stingers
 maxDistance=3000
 soundClassObject=&Fort_SFX_InGame
 volumeMultiplier=0.9
}
