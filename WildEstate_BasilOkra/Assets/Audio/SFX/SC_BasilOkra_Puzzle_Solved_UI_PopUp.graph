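# SoundCue for the BasilOkra puzzle-solved UI pop-up sound in WildEstate.
# A Delay node with a fixed short delay feeds a single WavePlayer that plays the pop-up wave.
# The cue routes to the Fort_UI_InGame_Critical sound class with a slight volume boost and a very large max distance.
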SoundCue /WildEstate_BasilOkra/Assets/Audio/SFX/SC_BasilOkra_Puzzle_Solved_UI_PopUp
root=Delay_0
nodes [
 {id=Delay_0, children [WavePlayer_1], properties {delayMax=0.3, delayMin=0.3}}
 {id=WavePlayer_1, properties {soundWaveAssetPtr=/WildEstate_BasilOkra/Assets/Audio/SFX/wav/BasilOkra_Puzzle_Solved_UI_PopUp}}
]
properties {maxDistance=2097152, soundClassObject=&Fort_UI_InGame_Critical, volumeMultiplier=1.2}
